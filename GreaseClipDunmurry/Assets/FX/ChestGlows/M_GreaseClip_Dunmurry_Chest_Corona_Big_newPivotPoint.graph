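# Additive, unlit, two-sided material for a large chest corona glow with animated rays.
# Counter-rotating spikeball flare samples, shaped by a sphere-masked halo gradient, drive a warm two-colour emissive blend with brightness controls.
# Opacity combines UV edge and falloff masks with an Anim parameter, plus depth and near-camera fades on the deferred path.
# World position offset rotates the mesh about its object pivot by camera-relative angle and applies per-axis size scaling.
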Material /GreaseClipDunmurry/Assets/FX/ChestGlows/M_GreaseClip_Dunmurry_Chest_Corona_Big_newPivotPoint
graph {
 properties {
  bUsedWithNanite=true
  blendMode=BLEND_Additive
  pixelDepthOffsetMode=PDOM_Legacy
  shadingModel=MSM_Unlit
  twoSided=true
  usedShadingModels=Unlit
 }
 nodes [
  n1 = ObjectOrientation()
  Abs(None=@n1)
  ActorPositionWS()
  n2 = Multiply(A=Time(), B=ScalarParameter() {defaultValue=0.25, parameterName=RotateSpeed})
  n3 = Rotator(Coordinate=Add(A=TextureCoordinate() {coordinateIndex=1, uTiling=0.25, vTiling=0.25}) {constB=0.375}, Time=@n2)
  n4 = Add(A=TextureCoordinate() {coordinateIndex=1, uTiling=0.15, vTiling=0.15}) {constB=0.425}
  n5 = OneMinus(None=TextureSample(UVs=Rotator(Coordinate=@n4, Time=@n2) {speed=-0.72}) {texture=/Game/Effects/Fort_Effects/Textures/Flares/T_Flare_Spikeball}.RGB)
  n6 = Saturate(None=Multiply(A=TextureSample(UVs=@n3) {texture=&T_Flare_Spikeball}.RGB, B=Power(Base=@n5)))
  n7 = ScalarParameter() {defaultValue=1, parameterName=Brightness}
  n8 = Constant() {r=0.5}
  n9 = SphereMask(
   A=TextureCoordinate() {coordinateIndex=1}
   B=@n8
   Radius=@n8
   Hardness=LinearInterpolate(Alpha=ScalarParameter() {defaultValue=0.95, parameterName=Size}) {constA=1, constB=0}
  ) {hardnessPercent=25}
  n10 = Multiply(
   A=VectorParameter() {parameterName=Corona_Colour, defaultValue=<rgba 9 3.7221 0 3>}.RGB
   B=Multiply(A=Multiply(A=@n6, B=@n7) {constB=80}, B=TextureSample(UVs=@n9) {texture=/Game/Effects/Fort_Effects/Textures/Grad/T_Halo_Corona_Gradient}.RGB)
  )
  n11 = MaterialFunctionCall(Float3=MaterialFunctionCall(LDR Input=Clamp(None=@n10), HDR Tint=@n10) {materialFunction=&DeriveHDRfromLDR}.RemappedRange) {materialFunction=&BreakOutFloat3Components}
  Add(A=@n11.R, B=@n11.R)
  n12 = Normalize(VectorInput=Subtract(A=ObjectPositionWS(), B=CameraPositionWS()))
  n13 = Arctangent2Fast(Y=ComponentMask(None=@n12) {g=true}, X=ComponentMask(None=@n12) {r=true})
  n14 = MaterialFunctionCall(Multiplier=Constant() {r=2}) {materialFunction=&Pi}
  n15 = RotateAboutAxis(
   NormalizedRotationAxis=@n1
   RotationAngle=Add(A=Divide(A=@n13, B=@n14.Result), B=Constant() {r=0.25}) {constA=0.25, constB=0.25}
   PivotPoint=MaterialFunctionCall() {materialFunction=&ObjectPivotPoint}.Object Pivot Location
   Position=WorldPosition().XYZ
  )
  n16 = ActorPositionWS()
  n17 = Add(A=@n15, B=WorldPosition() {worldPositionShaderOffset=WPT_ExcludeAllShaderOffsets}.XYZ)
  n18 = TransformPosition(None=@n17) {transformSourceType=TRANSFORMPOSSOURCE_World}
  Add(A=@n15, B=Subtract(A=Add(A=@n16, B=@n18), B=@n17)) {constB=0}
  Constant() {r=1}
  Constant()
  n19 = Max(A=Multiply(A=Power(Base=@n11.G) {constExponent=1.2}) {constB=0.8}, B=@n11.G)
  n20 = LinearInterpolate(
   A=VectorParameter() {parameterName=ColorWarm1, defaultValue=<rgba 0.8 0.1859 0.048 1>}.RGB
   B=VectorParameter() {parameterName=HighlightWarm1, defaultValue=<rgba 0.93 0.5106 0.1674 1>}.RGB
   Alpha=@n19
  )
  MaterialFunctionCall(Emissive=Multiply(A=Multiply(A=@n20, B=@n19), B=ScalarParameter() {defaultValue=1, parameterName=Scaled_Brightness})) {materialFunction=&MF_FX_EmissiveAdjustments}
  Max()
  n21 = MaterialFunctionCall(Float2=TextureCoordinate()) {materialFunction=&BreakOutFloat2Components}
  n22 = Sine(None=Add(A=Multiply(A=@n21.R) {constB=0.4}) {constB=0.05})
  n23 = Saturate(None=Multiply(A=Power(Base=@n22) {constExponent=4}))
  n24 = Constant() {r=1}
  n25 = DepthFade(
   Opacity=MaterialFunctionCall(Distance=ScalarParameter() {parameterName=NearCamFade}) {materialFunction=&Near_Camera_Fade}.Result
   FadeDistance=ScalarParameter() {defaultValue=50, parameterName=DepthFade}
  )
  n26 = ComponentMask(None=TextureCoordinate() {coordinateIndex=1}) {g=true}
  n27 = Multiply(
   A=Multiply(A=ScalarParameter() {defaultValue=3, parameterName=Corona_Opacity}, B=ShadingPathSwitch(Default=@n24, Deferred=@n25))
   B=Power(Base=OneMinus(None=@n26)) {constExponent=10}
  )
  n28 = ScalarParameter() {defaultValue=1, parameterName=Anim}
  NamedRerouteDeclaration(None=Multiply(A=Multiply(A=@n23) {constB=4}, B=Multiply(A=@n27, B=@n28)) {constA=1}) {name=Corona_Opacity, nodeColor=<rgba 1 0.2578 0 1>}
  n29 = Subtract(A=ObjectPositionWS(), B=WorldPosition().XYZ)
  n30 = OneMinus(None=VectorParameter() {parameterName=Scaled_MeshSizeXYZ, defaultValue=<rgba 1 1 1 1>}.RGB)
  n31 = Add(A=@n15, B=WorldPosition() {worldPositionShaderOffset=WPT_ExcludeAllShaderOffsets}.XYZ)
  n32 = Subtract(
   A=Add(
    A=MaterialFunctionCall() {materialFunction=&ObjectPivotPoint}.Object Pivot Location
    B=TransformPosition(None=@n31) {transformSourceType=TRANSFORMPOSSOURCE_World}
   )
   B=@n31
  )
  NamedRerouteDeclaration(None=Add(A=Add(A=@n15, B=@n32), B=Multiply(A=@n29, B=@n30))) {name=WPO, nodeColor=<rgba 0.3125 1 0 1>}
  NamedRerouteUsage()
  NamedRerouteUsage()
  ScalarParameter() {defaultValue=1, parameterName=Scaled_MeshSize}
  WorldPosition() {worldPositionShaderOffset=WPT_ExcludeAllShaderOffsets}
 ]
}
