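# Niagara system for a chest glow effect, with a burst-spawned ChestFlames emitter that renders fiery orange energy and glow through two mesh renderers and one sprite renderer.
# Spawn group selects lifetime and sets a visibility tag.
# Curves over life write separate namespaced color, size and orientation sets, plus dynamic-material vectors and a boosted fake-light color derived from the mesh color.
# Despite the system's name, the ChestSparkles emitter shows no modules or renderers.
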NiagaraSystem /GreaseClipDunmurry/Assets/FX/ChestGlows/NS_GreaseClip_Dunmurry_Chest_Sparkles
dependencies {
 usedDataInterfaces [&NiagaraDataInterfaceCurve, &NiagaraDataInterfaceColorCurve]
 usedDynamicInputs []
 usedModules [&SystemState]
}
emitters {
 ChestFlames {
  emitterUpdateScript [
   &EmitterState(Life Cycle Mode=::NewEnumerator1, Loop Behavior=::NewEnumerator1, Loop Duration=4, UseLoopDelay=false)
   &SpawnBurst_Instantaneous(Spawn Group=1, Use Spawn Probability=false, UseLoopCountLimit=true) {module=SpawnBurst_Instantaneous001}
  ]
  particleSpawnScript [
   /SetVariables(Particles.VisibilityTag=&SpawnGroup()) {isSetParameters=true}
   &InitializeParticle(
    Lifetime=&MergeHelper_FloatSwitch(0=0.5, 1=2.5, 2=?, 3=?, Index=&SpawnGroup(), Num=::NewEnumerator0)
    Mesh Scale Mode=::NewEnumerator3
    Position Offset=<0 0 50>
    Sprite Size=&Multiply_Vector2D_ByFloat(Float=600, Vector2D=<1 1>)
    Sprite Size Mode=::NewEnumerator4
    UsePositionOffset=false
    Write Material Random=true
   )
   &CameraOffset(Camera Offset Amount=7)
  ]
  particleUpdateScript [
   &ParticleState()
   &ScaleSpriteSize(Uniform Curve Sprite Scale=DataInterface() {propertyValues {Curve {keys [
        {arriveTangent=1.4378, interpMode=RCIM_Cubic, leaveTangent=1.4378, tangentMode=RCTM_User, time=0, value=0.45}
        {interpMode=RCIM_Cubic, tangentMode=RCTM_User, time=0.7, value=1}
       ]}}})
   &Color(Color=&ColorFromCurve(ColorCurve=DataInterface() {propertyValues {
       AlphaCurve {keys [<at 0 1 RCIM_Cubic>,<at 1 0 RCIM_Cubic>]}
       BlueCurve {keys [<at 0 2.0103 RCIM_Cubic>,<at 0.2559 0>,<at 1 0 RCIM_Cubic>]}
       GreenCurve {keys [<at 0 4.4861 RCIM_Cubic>,<at 0.2559 1.6445>,<at 1 0.1262 RCIM_Cubic>]}
       RedCurve {keys [<at 0 10 RCIM_Cubic>,<at 0.2559 5>,<at 1 1 RCIM_Cubic>]}
      }}))
   &FN_Sprite_Attributes(
    Color=&ColorFromCurve(ColorCurve=DataInterface() {propertyValues {
       AlphaCurve {keys [<at 0 1 RCIM_Cubic>,<at 0.8324 0 RCIM_Cubic>]}
       BlueCurve {keys [<at 0 0 RCIM_Cubic>,<at 0.8393 0 RCIM_Cubic>]}
       GreenCurve {keys [<at 0 3.289 RCIM_Cubic>,<at 0.8393 0.3598 RCIM_Cubic>]}
       RedCurve {keys [<at 0 8 RCIM_Cubic>,<at 0.8393 1 RCIM_Cubic>]}
      }})
    Create Color Attribute=true
    Create SpriteSize Attribute=true
    SpriteSize=&Vector2DFromFloat(Value=&FloatFromCurve(
      FloatCurve=DataInterface() {propertyValues {Curve {keys [
          {arriveTangent=2.5229, interpMode=RCIM_Cubic, leaveTangent=2.5229, tangentMode=RCTM_User, time=0, value=0}
          {interpMode=RCIM_Cubic, tangentMode=RCTM_User, time=0.85, value=1}
         ]}}}
      Scale Curve=550
     ))
    Write to Module Namespace=true
   ) {module=FN_Sprite_Attributes004}
   &FN_Mesh_Attributes(
    Color=&MultiplyLinearColor(
     Linear Color=&ColorFromCurve(ColorCurve=DataInterface() {propertyValues {
        AlphaCurve {keys [<at 0 1 RCIM_Cubic>]}
        BlueCurve {keys [<at 0 0.0143 RCIM_Cubic>,<at 0.4668 0.0259 RCIM_Cubic>]}
        GreenCurve {keys [<at 0 0.3205 RCIM_Cubic>,<at 0.4668 0.3418 RCIM_Cubic>]}
        RedCurve {keys [<at 0 1 RCIM_Cubic>,<at 0.4668 1 RCIM_Cubic>]}
       }})
     Scale Factor=&MakeFloat4FromVectorAndFloat(
      "Vector (RGB)"=&VectorFromFloat(Value=&FloatFromCurve(
        FloatCurve=DataInterface() {propertyValues {Curve {keys [
            {interpMode=RCIM_Cubic, tangentMode=RCTM_Break, time=0, value=1}
            {arriveTangent=-2.2558, interpMode=RCIM_Cubic, tangentMode=RCTM_Break, time=1, value=0}
           ]}}}
        Scale Curve=2
       ))
      "Float (Alpha)"=&FloatFromCurve(FloatCurve=DataInterface() {propertyValues {Curve {keys [<at 0.4 1 RCIM_Cubic>,<at 1 0>]}}}, Scale Curve=0)
     )
    )
    Create Color Attribute=true
    Create Orientation Attribute=true
    Create Scale Attribute=true
    Orientation:Quat4f=&MakeQuatFromAxisAngle(Axis=<0 1 0>, AngleInDegrees=90)
    Scale=&VectorFromFloat(Value=&FloatFromCurve(
      FloatCurve=DataInterface() {propertyValues {Curve {keys [
          {arriveTangent=2.0471, interpMode=RCIM_Cubic, leaveTangent=2.0471, tangentMode=RCTM_User, time=0, value=0}
          {interpMode=RCIM_Cubic, tangentMode=RCTM_User, time=1, value=1}
         ]}}}
      Scale Curve=4
     ))
    Write to Module Namespace=true
   ) {module=FN_Mesh_Attributes001}
   /SetVariables(Particles.EnergyMeshDynMat:Vector4f=&MakeVector4(
     X=&FloatFromCurve(FloatCurve=DataInterface() {propertyValues {Curve {keys [{arriveTangent=-2.1015, interpMode=RCIM_Cubic, leaveTangent=-2.1015, tangentMode=RCTM_User, time=0, value=0.5}, <at 1 -0.15>]}}})
     Y=&FloatFromCurve(FloatCurve=DataInterface() {propertyValues {Curve {keys [{arriveTangent=-0.3464, interpMode=RCIM_Cubic, leaveTangent=-0.3464, tangentMode=RCTM_User, time=0, value=1}, <at 1 0.9>]}}})
     Z=&FloatFromCurve(FloatCurve=DataInterface() {propertyValues {Curve {keys [
          {interpMode=RCIM_Cubic, tangentMode=RCTM_Break, time=0, value=0}
          {arriveTangent=0.4478, interpMode=RCIM_Cubic, leaveTangent=0.4478, tangentMode=RCTM_User, time=0.15, value=1}
          {interpMode=RCIM_Cubic, tangentMode=RCTM_Break, time=1, value=1.2}
         ]}}})
     W=0
    )) {isSetParameters=true}
   &FN_Mesh_Attributes(
    Color=&ColorFromCurve(ColorCurve=DataInterface() {propertyValues {
       AlphaCurve {keys [<at 0 0>,<at 0.191 1 RCIM_Cubic>,<at 1 0 RCIM_Cubic>]}
       BlueCurve {keys [<at 0.0091 2 RCIM_Cubic>,<at 0.1868 0.3>,<at 1 0.05 RCIM_Cubic>]}
       GreenCurve {keys [<at 0.0091 5.3333 RCIM_Cubic>,<at 0.1868 1.65>,<at 1 0.1625 RCIM_Cubic>]}
       RedCurve {keys [<at 0.0091 10 RCIM_Cubic>,<at 0.1868 3>,<at 1 0.5 RCIM_Cubic>]}
      }})
    Create Color Attribute=true
    Create Orientation Attribute=true
    Create Scale Attribute=true
    Orientation:Quat4f=&MakeQuatFromAxisAngle(Axis=<0 0 1>, AngleInDegrees=180)
    Scale=&Lerp_Vector(
     A=<16 7 3>
     B=<25 7.5 3.7>
     Alpha=&FloatFromCurve(FloatCurve=DataInterface() {propertyValues {Curve {keys [
          {arriveTangent=2.5848, interpMode=RCIM_Cubic, leaveTangent=2.5848, tangentMode=RCTM_User, time=0, value=0}
          {interpMode=RCIM_Cubic, tangentMode=RCTM_User, time=0.25, value=1}
          {interpMode=RCIM_Cubic, tangentMode=RCTM_User, time=1, value=0}
         ]}}})
    )
    Write to Module Namespace=true
   )
   /SetVariables(Particles.Chest_Energy_DynMat02:Vector4f=&MakeVector4(X=0, Y=1, Z=0, W=0)) {isSetParameters=true}
   /SetVariables(Particles.ChestEnergyPos:NiagaraPosition=<-10 0 -7>) {isSetParameters=true}
   /SetVariables(Particles.ColorFakeLight:LinearColor=&MultiplyLinearColor(
     Linear Color=$Particles.FN_Mesh_Attributes.Color
     Scale Factor=&MakeVector4FromVectorAndFloat(
      "Vector (RGB)"=&MakeVectorFromLinearColorRGB(Color=<rgba 50 22 10 1>)
      "Float (Alpha)"=&FloatFromCurve(FloatCurve=DataInterface() {propertyValues {Curve {keys [
           {arriveTangent=-1.5496, interpMode=RCIM_Cubic, leaveTangent=-1.5496, tangentMode=RCTM_User, time=0, value=1}
           {interpMode=RCIM_Cubic, tangentMode=RCTM_User, time=1, value=0}
          ]}}})
     )
    )) {isSetParameters=true}
  ]
  rendererClasses [&NiagaraMeshRendererProperties, &NiagaraSpriteRendererProperties]
  renderers [
   {rendererClass=&NiagaraMeshRendererProperties, rendererIndex=0}
   {rendererClass=&NiagaraMeshRendererProperties, rendererIndex=1}
   {rendererClass=&NiagaraSpriteRendererProperties, rendererIndex=2}
  ]
 }
 ChestSparkles {rendererClasses [], renderers []}
}
summary {
 systemName=NS_GreaseClip_Dunmurry_Chest_Sparkles
 emitters [
  {emitterName=ChestSparkles, rendererClasses []}
  {emitterName=ChestFlames, rendererClasses [&NiagaraMeshRendererProperties, &NiagaraSpriteRendererProperties]}
 ]
 userVariables [
  User.GlobalOffset {type=Vector3f, description="", defaultValue {struct=Vector3f, value=<0 0 0>}}
  User.u_Scale {type=NiagaraFloat, defaultValue=3, description=""}
 ]
}
